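; Passive gameplay ability for optic weapon mods that hides the weapon reticle while aiming down sights.
; On activation it caches the player's current weapon as B_Ranged_Generic and binds its targeting-changed and equipping-changed dispatchers.
; Targeting hides the reticle only with scope targeting in use and mobile auto-fire off, and otherwise shows it.
; Equip changes while not targeting, and ability end, restore the reticle.
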(event Ability|EventFailedToActivatePassiveAbility
  (params (ActorInfo :type Gameplay%20Ability%20Actor%20Info%20Structure))
  (Development|PrintString "Failed to activate"))

(event Ability|EventActivateAbility
  ;[1 1 1] Get Player's Current Weapon and Hide Reticle when ADS
  (bind _asb_ranged_generic (Utilities|Casting|CastToB_Ranged_Generic (Weapon|GetCurrentWeapon (CallFunction|GetAvatarasFortPlayerPawn))))
  (bind _output_get (Variables|Default|SetCurrentWeapon _asb_ranged_generic))
  (EventDispatchers|BindEventtoOnLocalTargetingChanged _output_get (delegate Custom|IsTargeting))
  (EventDispatchers|BindEventtoIsEquippingWeaponChanged _output_get (delegate Custom|EquipWeapon)))

(event Custom|EquipWeapon
  (params (bIsEquipping :type Boolean))
  (bind _currentweapon (Variables|Default|GetCurrentWeapon))
  ;[1 1 1] Get Player's Current Weapon and Hide Reticle when ADS
  (if (not (FortWeapon|IsTargeting _currentweapon))
    (Class|BRangedGeneric|ShowReticle _currentweapon)))

(event Custom|IsTargeting
  (params (bIsTargeting :type Boolean))
  (bind _currentweapon (Variables|Default|GetCurrentWeapon))
  ;[1 1 1] Get Player's Current Weapon and Hide Reticle when ADS
  (if (Math|Boolean|ANDBoolean bIsTargeting (not (Class|FortMobileKismetLibrary|GetAutoFireEnabled)) (Scope|UseScopeTargeting (Variables|Default|GetCurrentWeapon)))
    (Class|BRangedGeneric|HideReticle _currentweapon)
    (else
      (Class|BRangedGeneric|ShowReticle _currentweapon))))

(event Ability|EventOnEndAbility
  (params (bWasCancelled :type Boolean))
  ;[1 1 1] Show Reticle when Ability ends
  (Class|BRangedGeneric|ShowReticle (Variables|Default|GetCurrentWeapon)))
